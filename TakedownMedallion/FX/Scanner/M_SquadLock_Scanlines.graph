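# Additive two-sided material for Niagara particles, apparently the animated scanlines of the Squad Lock scanner effect in Takedown Medallion.
# Screen-space V samples a panning cracks texture, sharpened by repeated multiplication into horizontal bands.
# These add to a view-facing rim term from scene-texture world normals, with a constant normal on mobile and forward paths.
# Particle color tints the sum, and inverted DepthFade boosts intersections before MF_FX_EmissiveAdjustments.
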Material /TakedownMedallion/FX/Scanner/M_SquadLock_Scanlines
graph {
 properties {
  bUsedWithNiagaraMeshParticles=true
  bUsedWithNiagaraRibbons=true
  bUsedWithNiagaraSprites=true
  blendMode=BLEND_Additive
  pixelDepthOffsetMode=PDOM_Legacy
  twoSided=true
  usedShadingModels=Default Lit
 }
 nodes [
  Constant() {r=1}
  If()
  n1 = SceneTexture() {sceneTextureId=PPI_WorldNormal}
  n2 = Constant3Vector() {constant=<rgba 0 0 1 0>}
  n3 = ShadingPathSwitch(Default=@n1.Color, Deferred=@n1.Color, Forward=@n2, Mobile=@n2)
  n4 = ShadingPathSwitch(Default=@n3, Deferred=@n3, Mobile=@n2) {desc=Migrated from FeatureLevelSwitch}
  n5 = OneMinus(None=DotProduct(A=ComponentMask(None=@n4) {b=true, g=true, r=true}, B=CameraVectorWS()))
  n6 = Constant()
  n7 = Multiply(A=ComponentMask(None=ScreenPosition().ViewportUV) {g=true}) {constB=5}
  n8 = TextureSample(UVs=Panner(Coordinate=AppendVector(A=@n6, B=@n7)) {speedX=0.2569, speedY=0.6512}) {samplerType=SAMPLERTYPE_LinearColor, texture=/Game/Global/Textures/Default/Masks_and_Gradients/T_Cracks}
  n9 = Multiply(A=@n8.R, B=@n8.R)
  n10 = Multiply(A=ParticleColor().RGB, B=Add(A=Multiply(A=@n5) {constB=2}, B=Multiply(A=@n9, B=@n9)))
  n11 = Power(Base=OneMinus(None=DepthFade() {fadeDistanceDefault=20})) {constExponent=5}
  n12 = Multiply(A=@n10, B=@n11)
  MaterialFunctionCall(Emissive=Add(A=@n10, B=Multiply(A=@n12) {constB=8})) {materialFunction=&MF_FX_EmissiveAdjustments}
  Multiply(A=@n11) {constA=1, constB=0.3}
  n13 = TextureCoordinate()
  StaticSwitch(True=MaterialFunctionCall(XY=@n13) {materialFunction=&Swizzle}.YX)
  TextureSampleParameter2D(UVs=@n13) {parameterName=Param, samplerType=SAMPLERTYPE_Masks, texture=/Game/Global/Textures/Default/Noise/T_VerticalRipples_D}
 ]
}
