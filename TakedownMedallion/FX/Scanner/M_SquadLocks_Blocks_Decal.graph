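# Additive unlit Niagara sprite and ribbon material, named as a squad-lock scanner decal, that draws flickering, particle-colored digital blocks.
# Two data-chip noise channels pan at different speeds on stepped time and multiply, with dynamic parameters driving offset, UV scale and speed.
# A separate mask multiplies optional particle alpha, two curve-atlas UV ramps and inverted depth fade, which favors areas near intersecting geometry.
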Material /TakedownMedallion/FX/Scanner/M_SquadLocks_Blocks_Decal
graph {
 properties {
  bEnableResponsiveAA=true
  bUsedWithNiagaraRibbons=true
  bUsedWithNiagaraSprites=true
  blendMode=BLEND_Additive
  pixelDepthOffsetMode=PDOM_Legacy
  shadingModel=MSM_Unlit
  usedShadingModels=Unlit
 }
 nodes [
  n1 = ParticleColor()
  n2 = TextureCoordinate()
  n3 = MaterialFunctionCall(Float2=LinearInterpolate(
    A=MaterialFunctionCall(XY=@n2) {materialFunction=&Swizzle}.YX
    B=@n2
    Alpha=ScalarParameter() {defaultValue=1, parameterName=SwizzleUV, sliderMax=1}
   )) {materialFunction=&BreakOutFloat2Components}
  n4 = MaterialFunctionCall(
   X=Multiply(A=@n3.R, B=ScalarParameter() {defaultValue=1, parameterName=U_Scale, sliderMax=1})
   Y=Multiply(A=@n3.G, B=ScalarParameter() {defaultValue=1, parameterName=V_Scale, sliderMax=1})
  ) {materialFunction=&MakeFloat2}
  n5 = DynamicParameter() {parameterIndex=3, paramNames [TileOffset,TimeSpeed,BlockOpacity,Param4]}
  n6 = ScalarParameter() {parameterName=Pan_Speed_U, sliderMax=1}
  n7 = ScalarParameter() {parameterName=Pan_Speed_V, sliderMax=1}
  n8 = Multiply(
   A=Panner(Coordinate=Add(A=@n4.Result, B=@n5.TileOffset), Speed=AppendVector(A=@n6, B=@n7))
   B=DynamicParameter() {paramNames [UVScale,TimeSpeed,BlockOpacity,Param4]}.UVScale
  )
  n9 = Multiply(A=DynamicParameter() {paramNames [UVScale,TimeSpeed,BlockOpacity,Param4]}.TimeSpeed, B=Time())
  n10 = ScalarParameter() {defaultValue=6, parameterName=Steps}
  n11 = Divide(A=Ceil(None=Multiply(A=@n9, B=@n10)), B=@n10)
  n12 = VectorParameter() {parameterName=BlocksPanSpeed, defaultValue=<rgba 1.8 0.1 -0.8 -0.62>}
  n13 = AppendVector(A=@n12.R, B=@n12.R)
  n14 = AppendVector(A=@n12.B, B=@n12.B)
  n15 = Multiply(
   A=TextureSample(UVs=Panner(Coordinate=@n8, Time=@n11, Speed=@n13) {bFractionalPart=true}) {texture=/Game/Athena/UI/Challenges/Art/T_UI_DataChip_Noise}.R
   B=TextureSample(UVs=Panner(Coordinate=@n8, Time=@n11, Speed=@n14) {bFractionalPart=true}) {texture=&T_UI_DataChip_Noise}.G
  )
  MaterialFunctionCall(Emissive=Multiply(A=Multiply(A=@n1.RGB, B=@n15), B=ScalarParameter() {defaultValue=1, parameterName=ColorScale, sliderMax=1})) {materialFunction=&MF_FX_EmissiveAdjustments}
  n16 = Constant() {r=1}
  n17 = ScalarParameter() {parameterName=Use Particle Alpha}
  n18 = CurveAtlasRowParameter(CurveTime=MaterialFunctionCall(Float2=TextureCoordinate()) {materialFunction=&BreakOutFloat2Components}.G) {atlas=&/CA_SquadLockRamps, curve=&/C_VC_uRamp_A, parameterName=CA_Ramp_U}
  n19 = CurveAtlasRowParameter(CurveTime=MaterialFunctionCall(Float2=TextureCoordinate() {coordinateIndex=1}) {materialFunction=&BreakOutFloat2Components}.R) {
   atlas=&CA_SquadLockRamps
   curve=&/C_VC_vRamp_G
   defaultValue=2
   parameterName=CA_Ramp_V
  }
  n20 = ScalarParameter() {defaultValue=25, parameterName=FadeDistance, sliderMax=100}
  Multiply(A=Multiply(A=LinearInterpolate(A=@n16, B=@n1.A, Alpha=@n17), B=Multiply(A=@n18, B=@n19)), B=OneMinus(None=DepthFade(FadeDistance=@n20)))
  OneMinus(None=@n18)
 ]
}
