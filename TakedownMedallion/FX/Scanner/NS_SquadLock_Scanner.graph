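# NS_SquadLock_Scanner is a blue Niagara scanner effect for the SquadLock feature of the Takedown Medallion.
# Upward-facing sprite planes oscillate within User.BboxSize, with spawning gated by loop age.
# A two-particle ribbon forms a beam from User.ScannerCenter to the oscillating owner position, widening with distance.
# User.AgeInput scales particle age on the main planes and ribbon, while onEnter emitters add denser planes and a blocky flash.
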NiagaraSystem /TakedownMedallion/FX/Scanner/NS_SquadLock_Scanner
dependencies {
 usedDataInterfaces [&NiagaraDataInterfaceCurve, &NiagaraDataInterfaceColorCurve]
 usedDynamicInputs []
 usedModules [&SystemState]
}
emitters {
 BlockyStuff_onEnter {
  emitterUpdateScript [
   &EmitterState(Inactive Response=::NewEnumerator1, Life Cycle Mode=::NewEnumerator1, Loop Behavior=::NewEnumerator1, Loop Duration=1.5, UseLoopDelay=false)
   &SpawnBurst_Instantaneous(Use Spawn Probability=false, UseLoopCountLimit=false)
  ]
  particleSpawnScript [&InitializeParticle(
    Color=<rgba 0 0.2 1 1>
    Lifetime=1.5
    Sprite Size Mode=::NewEnumerator3
    Uniform Sprite Size=111
    UsePositionOffset=false
    Write Material Random=true
   )]
  particleUpdateScript [
   &ParticleState()
   /SetVariables(Particles.Position=&Add_Vector(A=<0 0 0>, B=<0 0 25>)) {isSetParameters=true}
   /SetVariables(Particles.SpriteSize=&Multiply_Vector2D_ByFloat(
     Float=&FloatFromCurve(
      FloatCurve=DataInterface() {propertyValues {Curve {keys [
          {arriveTangent=-0.0094, interpMode=RCIM_Cubic, leaveTangent=-0.0094, tangentMode=RCTM_User, time=0, value=2}
          {arriveTangent=3.9425, interpMode=RCIM_Cubic, leaveTangent=3.9425, tangentMode=RCTM_User, time=0.3, value=3.25}
          {arriveTangent=-0.0405, interpMode=RCIM_Cubic, leaveTangent=-0.0405, tangentMode=RCTM_User, time=1, value=4.375}
         ]}}}
      Scale Curve=15
     )
     Vector2D=<1 1>
    )) {isSetParameters=true}
   &ScaleColor(
    Scale Alpha=&FloatFromCurve(
     FloatCurve=DataInterface() {propertyValues {Curve {keys [
         {arriveTangent=-0.2933, interpMode=RCIM_Cubic, leaveTangent=-0.2933, tangentMode=RCTM_User, time=0, value=0}
         {arriveTangent=-2.8313, interpMode=RCIM_Cubic, leaveTangent=-2.8313, tangentMode=RCTM_User, time=0.1, value=0.95}
         {arriveTangent=-0.9489, interpMode=RCIM_Cubic, leaveTangent=-0.9489, tangentMode=RCTM_User, time=0.2, value=0.25}
         {arriveTangent=-0.0383, interpMode=RCIM_Cubic, leaveTangent=-0.0383, tangentMode=RCTM_User, time=1, value=0}
        ]}}}
     Scale Curve=2
    )
    Scale RGB=&Multiply_VectorByFloat(
     Vector=<1 1 1>
     Float=&FloatFromCurve(FloatCurve=DataInterface() {propertyValues {Curve {keys [
          {arriveTangent=-0.2933, interpMode=RCIM_Cubic, leaveTangent=-0.2933, leaveTangentWeight=0.3333, tangentMode=RCTM_User, time=0, value=0}
          {
           arriveTangent=0.0608
           arriveTangentWeight=0.0333
           interpMode=RCIM_Cubic
           leaveTangent=0.0608
           leaveTangentWeight=0.3
           tangentMode=RCTM_User
           time=0.0594
           value=4.75
          }
          {arriveTangent=-27.9094, interpMode=RCIM_Cubic, leaveTangent=-27.9094, tangentMode=RCTM_User, time=0.085, value=4.0956}
          {
           arriveTangent=-0.6853
           arriveTangentWeight=0.0333
           interpMode=RCIM_Cubic
           leaveTangent=-0.6853
           leaveTangentWeight=0.2667
           tangentMode=RCTM_User
           time=0.5036
           value=0.3433
          }
          {arriveTangent=-0.8945, arriveTangentWeight=0.2667, interpMode=RCIM_Cubic, leaveTangent=-0.8945, tangentMode=RCTM_User, time=1, value=0}
         ]}}})
    )
    ScaleA=true
    ScaleRGB=true
   ) {module=ScaleColor001}
   &CameraOffset(Camera Offset Amount=25) {module=CameraOffset001}
   /SetVariables(
    Particles.Color_B:LinearColor=&MultiplyLinearColor(Linear Color=$Particles.Color, Scale Factor=&MakeVector4_FromFloat(Float=0.3))
    Particles.SpriteSize_B:Vector2f=&Multiply_Vector2D_ByFloat(Float=1, Vector2D=$Particles.SpriteSize)
   ) {enabled=false, isSetParameters=true}
  ]
  rendererClasses [&NiagaraSpriteRendererProperties]
  renderers [{rendererClass=&NiagaraSpriteRendererProperties}]
 }
 Scan_Planes {
  emitterUpdateScript [
   &EmitterState(Life Cycle Mode=::NewEnumerator1, Loop Behavior=::NewEnumerator1, Loop Duration=1, UseLoopDelay=false)
   &SpawnRate(
    Spawn Probability=&MakeCustomFloatFromBool(
     Bool=&SetBoolByFloatComparison(A=$Emitter.NormalizedLoopAge, B=0.3, Comparison Type=::NewEnumerator6)
     True Float=1
     False Float=0
    )
    SpawnRate=20
    Use Spawn Probability=true
   )
  ]
  particleSpawnScript [
   &InitializeParticle(
    Color=<rgba 0 0.3 3 1>
    Lifetime=0.5
    Mass Mode=::NewEnumerator1
    Position Mode=::NewEnumerator1
    Sprite Size=<75 75>
    Sprite Size Mode=::NewEnumerator4
    UsePositionOffset=true
    Write Material Random=true
   )
   /SetVariables(Particles.Position=&Add_Vector(A=<0 0 0>, B=<0 0 55>)) {isSetParameters=true, module=SetVariables_#1}
  ]
  particleUpdateScript [
   &ParticleState()
   /SetVariables(Particles.Age:NiagaraFloat=&Multiply_Float(A=$Particles.Age, B=&Multiply_Float(A=$User.AgeInput, B=5))) {isSetParameters=true, module=SetVariables_#2}
   &ScaleColor(
    Scale Alpha=&FloatFromCurve(FloatCurve=DataInterface() {propertyValues {Curve {keys [
         {arriveTangent=8.6044, interpMode=RCIM_Cubic, leaveTangent=8.6044, leaveTangentWeight=0.0349, tangentMode=RCTM_User, time=-0.0046, value=0}
         <at 0.1 0.9>
         {
          arriveTangent=-1.3396
          arriveTangentWeight=0.2
          interpMode=RCIM_Cubic
          leaveTangent=-1.3396
          leaveTangentWeight=0.0666
          tangentMode=RCTM_User
          time=0.5
          value=0.35
         }
         {arriveTangent=0.0723, interpMode=RCIM_Cubic, leaveTangent=0.0723, tangentMode=RCTM_User, time=1.0011, value=0.002}
        ]}}})
    ScaleA=true
    ScaleRGB=true
   )
   &SpriteFacingAndAlignment(Facing Coordinate Space=::World, Sprite Facing=<0 0 1>, WriteAlignment=false, WriteFacing=true)
   &self:OscillateInBounds(
    BboxSize:Vector3f=$User.BboxSize
    InitialPosition:Vector3f=<0 0 35>
    InputTime:NiagaraFloat=$Emitter.Age
    Magnitude=1
    Speed:NiagaraFloat=&RandomRangeFloat(Minimum=0.1, Maximum=0.4)
    Stagger:NiagaraFloat=&ReturnNormalizedExecIndex(Normalized Index Scale=0.55)
   ) {module=OscillateInBounds}
   &CameraOffset(Camera Offset Amount=15)
   &Color(Color=&LerpLinearColors(
     StartColor=<rgba 0 0.05 0.5 1>
     EndColor=<rgba 0 0 1 1>
     LerpFactor=&FloatFromCurve(FloatCurve=DataInterface() {propertyValues {Curve {keys [<at 0.1 0>,<at 0.3 1>]}}})
    ))
   /SetVariables(Particles.Color=&ScaleLinearColorByCurve(
     InitialValue=$Particles.Color
     ScaleCurve=DataInterface() {propertyValues {
       AlphaCurve {keys []}
       BlueCurve {keys [<at 0 1>,<at 0.0441 1>,<at 0.5584 0.1042>,<at 1 0>]}
       GreenCurve {keys [<at 0 1>,<at 0.0441 1>,<at 0.5584 0.1042>,<at 1 0>]}
       RedCurve {keys [<at 0 1>,<at 0.0441 1>,<at 0.5584 0.1042>,<at 1 0>]}
      }}
     CurveIndex=$Particles.Oscillate
    )) {isSetParameters=true, module=SetVariables_#3}
  ]
  rendererClasses [&NiagaraSpriteRendererProperties]
  renderers [{rendererClass=&NiagaraSpriteRendererProperties}]
 }
 Scan_Planes_onEnter {
  emitterUpdateScript [
   &EmitterState(Life Cycle Mode=::NewEnumerator1, Loop Behavior=::NewEnumerator1, Loop Duration=1, UseLoopDelay=false)
   &SpawnRate(
    Spawn Probability=&MakeCustomFloatFromBool(
     Bool=&SetBoolByFloatComparison(A=$Emitter.NormalizedLoopAge, B=0.3, Comparison Type=::NewEnumerator6)
     True Float=1
     False Float=0
    )
    SpawnRate=55
    Use Spawn Probability=true
   )
  ]
  particleSpawnScript [
   &InitializeParticle(
    Color Mode=::NewEnumerator0
    Mass Mode=::NewEnumerator1
    Position Mode=::NewEnumerator1
    Sprite Size=<75 75>
    Sprite Size Mode=::NewEnumerator4
    UsePositionOffset=true
    Write Material Random=true
   )
   /SetVariables(Particles.Position=&Add_Vector(A=<0 0 0>, B=<0 0 55>)) {isSetParameters=true, module=SetVariables_#1}
  ]
  particleUpdateScript [
   &ParticleState()
   &ScaleColor(
    Scale Alpha=&FloatFromCurve(FloatCurve=DataInterface() {propertyValues {Curve {keys [
         {arriveTangent=8.6044, interpMode=RCIM_Cubic, leaveTangent=8.6044, leaveTangentWeight=0.0349, tangentMode=RCTM_User, time=-0.0046, value=0}
         <at 0.1 0.9>
         {
          arriveTangent=-1.3396
          arriveTangentWeight=0.2
          interpMode=RCIM_Cubic
          leaveTangent=-1.3396
          leaveTangentWeight=0.0666
          tangentMode=RCTM_User
          time=0.5
          value=0.35
         }
         {arriveTangent=0.0723, interpMode=RCIM_Cubic, leaveTangent=0.0723, tangentMode=RCTM_User, time=1.0011, value=0.002}
        ]}}})
    ScaleA=true
    ScaleRGB=true
   )
   &SpriteFacingAndAlignment(Facing Coordinate Space=::World, Sprite Facing=<0 0 1>, WriteAlignment=false, WriteFacing=true)
   &self:OscillateInBounds(
    BboxSize:Vector3f=$User.BboxSize
    InitialPosition:Vector3f=<0 0 35>
    InputTime:NiagaraFloat=$Emitter.Age
    Magnitude=1
    Speed:NiagaraFloat=&RandomRangeFloat(Minimum=0.1, Maximum=0.4)
    Stagger:NiagaraFloat=&ReturnNormalizedExecIndex(Normalized Index Scale=0.55)
   ) {module=OscillateInBounds}
   &CameraOffset(Camera Offset Amount=15)
   &Color(Color=&LerpLinearColors(
     StartColor=<rgba 0 0 1 1>
     EndColor=<rgba 0 0.01 0 1>
     LerpFactor=&FloatFromCurve(FloatCurve=DataInterface() {propertyValues {Curve {keys [
          {arriveTangent=0.1837, interpMode=RCIM_Cubic, leaveTangent=0.1837, tangentMode=RCTM_User, time=0, value=0}
          {arriveTangent=-0.2496, interpMode=RCIM_Cubic, leaveTangent=-0.2496, tangentMode=RCTM_User, time=1, value=1}
         ]}}})
    ))
   /SetVariables(Particles.Color=&ScaleLinearColorByCurve(
     InitialValue=$Particles.Color
     ScaleCurve=DataInterface() {propertyValues {
       AlphaCurve {keys []}
       BlueCurve {keys [<at 0 1>,<at 0.0441 1>,<at 0.5584 0.1042>,<at 1 0>]}
       GreenCurve {keys [<at 0 1>,<at 0.0441 1>,<at 0.5584 0.1042>,<at 1 0>]}
       RedCurve {keys [<at 0 1>,<at 0.0441 1>,<at 0.5584 0.1042>,<at 1 0>]}
      }}
     CurveIndex=$Particles.Oscillate
    )) {isSetParameters=true, module=SetVariables_#3}
  ]
  rendererClasses [&NiagaraSpriteRendererProperties]
  renderers [{rendererClass=&NiagaraSpriteRendererProperties}]
 }
 Scan_Ribbon {
  emitterUpdateScript [
   &EmitterState(Life Cycle Mode=::NewEnumerator1, Loop Behavior=::NewEnumerator1, Loop Duration=1, UseLoopDelay=false)
   &SpawnBurst_Instantaneous(Spawn Count=2, Spawn Group=1, Use Spawn Probability=false, UseLoopCountLimit=false) {module=SpawnBurst_Instantaneous001}
  ]
  particleSpawnScript [
   &InitializeRibbon(
    Color=<rgba 1 0 0.6382 1>
    Lifetime=0.5
    Position Mode:ENiagara_PositionInitializationMode=::NewEnumerator0
    Ribbon Facing Mode:ENiagara_UnsetDirectSet=::NewEnumerator1
    Ribbon Width=125
    Write Material Random=false
   )
   /SetVariables(Particles.ScannerCenter:Vector3f=&Add_Vector(A=$User.ScannerCenter, B=<25 5 3>)) {isSetParameters=true}
  ]
  particleUpdateScript [
   &ParticleState()
   /SetVariables(Particles.Age:NiagaraFloat=&Multiply_Float(A=$Particles.Age, B=&Multiply_Float(A=$User.AgeInput, B=5))) {isSetParameters=true, module=SetVariables_#2}
   &self:OscillatePosition(
    BboxSize:Vector3f=$User.BboxSize
    InitialPosition:Vector3f=<0 0 44>
    InputTime:NiagaraFloat=$Emitter.Age
    Magnitude=0.2
    Speed=0.3
    Stagger:NiagaraFloat=&ReturnNormalizedExecIndex(Normalized Index Scale=0.55)
   ) {module=OscillatePosition}
   /SetVariables(Particles.Position=&Lerp_Vector(
     A=$Particles.ScannerCenter
     B=&Add_Vector(A=Linked() {linkedVariable {name=Engine.Owner.Position, type=NiagaraPosition}}, B=$Particles.OscillatePosition)
     Alpha=&ReturnNormalizedExecIndex()
    )) {isSetParameters=true}
   /SetVariables(Particles.RibbonWidth=&Lerp_Float(
     A=0
     B=&Multiply_Float(
      A=&VectorLength(Vector=&Subtract_Vector(A=$Particles.ScannerCenter, B=Linked() {linkedVariable {name=Engine.Owner.Position, type=NiagaraPosition}}))
      B=0.35
     )
     Alpha=&ReturnNormalizedExecIndex()
    )) {isSetParameters=true}
   &Color(Color=&LerpLinearColors(
     StartColor=<rgba 0 0.02 0.1 0.3>
     EndColor=<rgba 0 0 0.3 1>
     LerpFactor=&FloatFromCurve(FloatCurve=DataInterface() {propertyValues {Curve {keys [<at 0.1 0>,<at 0.3 1>]}}})
    ))
  ]
  rendererClasses [&NiagaraRibbonRendererProperties]
  renderers [{rendererClass=&NiagaraRibbonRendererProperties}]
 }
}
summary {
 systemName=NS_SquadLock_Scanner
 emitters [
  {emitterName=Scan_Planes, rendererClasses [&NiagaraSpriteRendererProperties]}
  {emitterName=Scan_Ribbon, rendererClasses [&NiagaraRibbonRendererProperties]}
  {emitterName=Scan_Planes_onEnter, rendererClasses [&NiagaraSpriteRendererProperties]}
  {emitterName=BlockyStuff_onEnter, rendererClasses [&NiagaraSpriteRendererProperties]}
 ]
 userVariables [
  User.ActorCenter {type=Vector3f, description="", defaultValue {struct=Vector3f, value=<0 0 0>}}
  User.AgeInput {type=NiagaraFloat, defaultValue=0, description=""}
  User.BboxSize {type=Vector3f, description="", defaultValue {struct=Vector3f, value=<0 0 55>}}
  User.ScannerCenter {type=Vector3f, description="", defaultValue {struct=Vector3f, value=<0 0 0>}}
 ]
}
